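# Material function for countdown displays that scales a 0-1 2D position to account for aspect ratio.
# It multiplies X by the XY ratio and Y by the YX ratio, each ratio floored at 1 by Max, so only the longer axis stretches.
# A static switch can derive YX as the reciprocal of XY instead of reading the supplied value, and is off by default.
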MaterialFunction /GrillMoon_Countdown2/Assets/TechArt/CountdownDisplays/Materials/MF_Display_PositionFromAspectRatio
graph {nodes [
  n1 = FunctionInput() {inputName=0-1 Position, inputType=FunctionInput_Vector2}
  n2 = ComponentMask(None=@n1) {r=true}
  n3 = FunctionInput() {inputName="(XY, YX) Ratios", inputType=FunctionInput_Vector2, sortPriority=1}
  n4 = ComponentMask(None=@n3) {r=true}
  n5 = Max(A=@n4)
  n6 = ComponentMask(None=@n1) {g=true}
  n7 = StaticBool()
  n8 = Max(A=StaticSwitch(
    True=Divide(B=@n4)
    False=ComponentMask(None=@n3) {g=true}
    Value=FunctionInput(Preview=@n7) {bUsePreviewValueAsDefault=true, inputName=Reconstruct YX From XY, inputType=FunctionInput_StaticBool, sortPriority=2}
   ))
  FunctionOutput(None=AppendVector(A=Multiply(A=@n2, B=@n5), B=Multiply(A=@n6, B=@n8))) {outputName=Scaled Position}
 ]}
