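# Material function for GrillMoon countdown display materials that remaps the U coordinate in a box corner and outputs CornerX and CornerMask.
# Ceiled, saturated UV components multiply into the mask for the positive quadrant.
# Corner width is aspect-corrected per axis via XYRatio, YXRatio and an X>Y blend.
# V's ramp across that width offsets U, masked to the corner; an appended full-UV result feeds no output.
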MaterialFunction /GrillMoon_Countdown2/Assets/TechArt/CountdownDisplays/Materials/MF_GrillMoon_BoxCornerBuilder
graph {nodes [
  n1 = FunctionInput() {inputName=CornerUV, inputType=FunctionInput_Vector2}
  n2 = ComponentMask(None=@n1) {r=true}
  n3 = Saturate(None=@n2)
  n4 = ComponentMask(None=@n1) {g=true}
  n5 = Saturate(None=@n4)
  n6 = Multiply(A=Ceil(None=@n3), B=Ceil(None=@n5))
  n7 = OneMinus(None=@n6)
  n8 = FunctionInput() {inputName="Input Corner Width (Scalar)", inputType=FunctionInput_Scalar, sortPriority=10}
  n9 = FunctionInput() {inputName=XYRatio, inputType=FunctionInput_Scalar, sortPriority=40}
  n10 = FunctionInput() {inputName="X>Y", inputType=FunctionInput_Scalar, sortPriority=30}
  n11 = Clamp(None=Multiply(A=@n8, B=LinearInterpolate(A=@n9, Alpha=@n10))) {minDefault=0.001}
  n12 = Saturate(None=InverseLinearInterpolate(A=OneMinus(None=@n11), Value=@n5))
  n13 = Multiply(A=@n8, B=LinearInterpolate(B=FunctionInput() {inputName=YXRatio, inputType=FunctionInput_Scalar, sortPriority=20}, Alpha=@n10) {constA=1})
  n14 = Multiply(A=Add(A=@n3, B=Multiply(A=@n12, B=@n13)), B=@n6)
  AppendVector(A=Add(A=Multiply(A=@n2, B=@n7), B=@n14), B=@n4)
  FunctionOutput(None=@n14) {outputName=CornerX}
  FunctionOutput(None=@n6) {outputName=CornerMask}
 ]}
