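# MF_GrillMoon_DisplayElementMixer is a material function that layers a new element over a base for countdown display materials.
# Base color is multiplied by the inverse of the new element mask, then the new color is added unweighted, as in premultiplied-alpha compositing.
# The combined mask is the maximum of both masks, with the base mask defaulting to 1 when unspecified.
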MaterialFunction /GrillMoon_Countdown2/Assets/TechArt/CountdownDisplays/Materials/MF_GrillMoon_DisplayElementMixer
graph {nodes [
  n1 = FunctionInput() {inputName=Base Element Color}
  n2 = FunctionInput() {inputName=New Element Mask, inputType=FunctionInput_Scalar, sortPriority=3}
  n3 = OneMinus(None=@n2)
  FunctionOutput(None=Add(A=Multiply(A=@n1, B=@n3), B=FunctionInput() {inputName=New Element Color, sortPriority=2})) {outputName=Combined Color}
  n4 = FunctionInput(Preview=Constant() {r=1}) {
   bUsePreviewValueAsDefault=true
   description=Set to 1 if not specified
   inputName=Base Element Mask
   inputType=FunctionInput_Scalar
   sortPriority=1
  }
  FunctionOutput(None=Max(A=@n4, B=@n2)) {outputName=Combined Mask, sortPriority=1}
 ]}
