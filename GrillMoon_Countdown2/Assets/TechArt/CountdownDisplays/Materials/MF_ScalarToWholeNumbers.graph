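# Material function that renders a scalar as a whole number of up to eight digits from a digit-strip texture, for countdown display materials.
# A modulo and divide chain splits the floored value into digits, and nested Ifs pick one per UV cell, sized to the digit count.
# Padded UVs index the strip, sampled with original-UV derivatives despite the Frac discontinuities.
# It outputs masked greyscale, color and a 0-1 mask.
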MaterialFunction /GrillMoon_Countdown2/Assets/TechArt/CountdownDisplays/Materials/MF_ScalarToWholeNumbers
graph {nodes [
  n1 = FunctionInput(Preview=TextureCoordinate() {coordinateIndex=2}) {bUsePreviewValueAsDefault=true, inputName=UVs, inputType=FunctionInput_Vector2, sortPriority=1}
  n2 = ScalarParameter() {defaultValue=0.9128, group=Numbers, parameterName=u_NumberPaddingScale}
  n3 = ScalarParameter() {defaultValue=0.3143, group=Numbers, parameterName=v_NumberPaddingScale}
  n4 = MaterialFunctionCall(
   UVs=@n1
   Texture Scale=FunctionInput(Preview=AppendVector(A=@n2, B=@n3)) {bUsePreviewValueAsDefault=true, inputName=UVPadding, inputType=FunctionInput_Vector2, sortPriority=1}
  ) {materialFunction=&ScaleUVsByCenter}
  n5 = ComponentMask(None=@n4.UVs) {r=true}
  n6 = FunctionInput(Preview=Constant() {desc=Number of digits, r=8}) {
   bUsePreviewValueAsDefault=true
   desc=Only whole numbers between 1 and 8 Using a constant saves some instructions
   description=Using a constant saves some instructions
   inputName="NumberofDigits(int)"
   inputType=FunctionInput_Scalar
   sortPriority=2
  }
  n7 = Constant() {r=1}
  n8 = Constant() {r=4}
  n9 = Constant() {r=7}
  n10 = ComponentMask(None=Multiply(A=OneMinus(None=@n5), B=@n6) {constB=5}) {r=true}
  n11 = NamedRerouteUsage()
  n12 = If(A=Constant() {r=1}, B=@n10, "A > B"=@n11, "A < B"=NamedRerouteUsage())
  n13 = If(A=Constant() {r=2}, B=@n10, "A > B"=@n12, "A < B"=NamedRerouteUsage())
  n14 = If(A=Constant() {r=3}, B=@n10, "A > B"=@n13, "A < B"=NamedRerouteUsage())
  n15 = If(A=Constant() {r=4}, B=@n10, "A > B"=@n14, "A < B"=NamedRerouteUsage())
  n16 = If(A=Constant() {r=5}, B=@n10, "A > B"=@n15, "A < B"=NamedRerouteUsage())
  n17 = If(A=Constant() {r=6}, B=@n10, "A > B"=@n16, "A < B"=NamedRerouteUsage())
  n18 = If(A=Constant() {r=7}, B=@n10, "A > B"=@n17, "A < B"=NamedRerouteUsage())
  n19 = If(A=@n6, B=Constant() {r=8}, "A > B"=@n18, "A == B"=@n18, "A < B"=@n18)
  n20 = If(
   A=@n6
   B=Constant() {r=5}
   "A > B"=If(A=@n6, B=Constant() {r=6}, "A > B"=If(A=@n6, B=@n9, "A > B"=@n19, "A == B"=@n17, "A < B"=@n17), "A == B"=@n16, "A < B"=@n16)
   "A == B"=@n15
   "A < B"=@n15
  )
  n21 = If(
   A=@n6
   B=Constant() {r=2}
   "A > B"=If(A=@n6, B=Constant() {r=3}, "A > B"=If(A=@n6, B=@n8, "A > B"=@n20, "A == B"=@n14, "A < B"=@n14), "A == B"=@n13, "A < B"=@n13)
   "A == B"=@n12
   "A < B"=@n12
  )
  n22 = Add(A=Frac(None=Multiply(A=@n5, B=@n6)), B=Floor(None=If(A=@n6, B=@n7, "A > B"=@n21, "A == B"=@n11, "A < B"=@n11)))
  n23 = ScalarParameter() {defaultValue=11, group=Numbers, parameterName=Numbers of Digits On Texture}
  n24 = TextureSample!(
   UVs=AppendVector(A=Divide(A=@n22, B=@n23) {constB=11}, B=ComponentMask(None=@n4.UVs) {g=true})
   Tex=FunctionInput(Preview=TextureObject() {samplerType=SAMPLERTYPE_Masks, texture=/GrillMoon_Countdown2/Assets/TechArt/CountdownDisplays/Textures/T_Numbers2_M}) {bUsePreviewValueAsDefault=true, inputName=Number Texture, inputType=FunctionInput_Texture2D}
   "DDX(UVs)"=DDX(Value=@n1)
   "DDY(UVs)"=DDY(Value=@n1)
   Apply View MipBias
  ) {mipValueMode=TMVM_Derivative}
  FunctionOutput(None=Multiply(A=@n24.R, B=@n4.0-1 mask)) {outputName=NumbersGreyscale}
  FunctionOutput(None=@n4.0-1 mask) {outputName=0-1 Mask, sortPriority=2}
  FunctionOutput(None=@n24.RGB) {outputName=NumbersColor, sortPriority=1}
  n25 = Floor(None=FunctionInput(Preview=Constant() {r=1234567}) {bUsePreviewValueAsDefault=true, inputName="NumbersToDisplay(int)", inputType=FunctionInput_Scalar})
  n26 = Constant() {r=10}
  n27 = Modulo!(A=@n25, B=@n26)
  n28 = Divide(A=Subtract(A=@n25, B=NamedRerouteDeclaration(None=@n27) {name=Digit_1, nodeColor=<rgba 0.8984 0 1 1>}), B=@n26)
  n29 = Modulo(A=@n28, B=@n26)
  n30 = Divide(A=Subtract(A=@n28, B=NamedRerouteDeclaration(None=@n29) {name=Digit_2, nodeColor=<rgba 0.8984 0 1 1>}), B=@n26)
  n31 = Modulo(A=@n30, B=@n26)
  n32 = Divide(A=Subtract(A=@n30, B=NamedRerouteDeclaration(None=@n31) {name=Digit_3, nodeColor=<rgba 0.8984 0 1 1>}), B=@n26)
  n33 = Modulo(A=@n32, B=@n26)
  n34 = Divide(A=Subtract(A=@n32, B=NamedRerouteDeclaration(None=@n33) {name=Digit_4, nodeColor=<rgba 0.8984 0 1 1>}), B=@n26)
  n35 = Modulo(A=@n34, B=@n26)
  n36 = Divide(A=Subtract(A=@n34, B=NamedRerouteDeclaration(None=@n35) {name=Digit_5, nodeColor=<rgba 0.8984 0 1 1>}), B=@n26)
  n37 = Modulo(A=@n36, B=@n26)
  n38 = Divide(A=Subtract(A=@n36, B=NamedRerouteDeclaration(None=@n37) {name=Digit_6, nodeColor=<rgba 0.8984 0 1 1>}), B=@n26)
  n39 = Modulo(A=@n38, B=@n26)
  n40 = Divide(A=Subtract(A=@n38, B=NamedRerouteDeclaration(None=@n39) {name=Digit_7, nodeColor=<rgba 0.8984 0 1 1>}), B=@n26)
  Subtract(A=@n40, B=NamedRerouteDeclaration(None=Modulo(A=@n40, B=@n26)) {name=Digit_8, nodeColor=<rgba 0.8984 0 1 1>})
 ]}
